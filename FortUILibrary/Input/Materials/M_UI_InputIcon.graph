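# Translucent UI material that draws input prompt icons, such as key glyphs, from packed distance-field textures.
# The Icon texture's red channel forms the shell and green the glyph, both anti-aliased by MF_UI_AntiAliasing, with an eased outline shadow.
# Monochromatic and Use Color switches plus a dark-mode parameter choose colors, tinted by vertex color.
# Use Override swaps the procedural color and opacity for a plain Icon Override texture.
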Material /FortUILibrary/Input/Materials/M_UI_InputIcon
graph {
 properties {
  bAutomaticallySetUsageInEditor=false
  bUsedWithStaticMesh=false
  blendMode=BLEND_Translucent
  materialDomain=MD_UI
  pixelDepthOffsetMode=PDOM_Legacy
  shadingModel=MSM_Unlit
  twoSided=true
  usedShadingModels=Unlit
 }
 nodes [
  n1 = NamedRerouteUsage()
  n2 = NamedRerouteUsage()
  n3 = NamedRerouteUsage()
  n4 = VectorParameter() {parameterName=Color, defaultValue=<rgba 1 0 0.5024 1>}
  n5 = NamedRerouteUsage()
  n6 = StaticSwitchParameter(
   True=NamedRerouteUsage()
   False=StaticSwitchParameter(True=LinearInterpolate(A=@n3, B=@n4.RGB, Alpha=@n5), False=NamedRerouteUsage()) {parameterName=Use Color}
  ) {parameterName=Monochromatic}
  n7 = StaticSwitchParameter(
   True=LinearInterpolate(
    A=VectorParameter() {parameterName=Light Mode Color Override, defaultValue=<rgba 1 0.8344 0 1>}.RGB
    B=NamedRerouteUsage()
    Alpha=NamedRerouteUsage()
   )
   False=NamedRerouteUsage()
  ) {parameterName=Use Color}
  n8 = NamedRerouteUsage()
  n9 = LinearInterpolate(A=@n6, B=@n7, Alpha=@n8)
  n10 = ScalarParameter() {defaultValue=0.5, parameterName=Shell Brightness, sliderMax=1}
  n11 = Multiply(A=Subtract(A=NamedRerouteUsage()) {constB=0.5}, B=ScalarParameter() {defaultValue=0.875, parameterName=Scale, sliderMax=1}) {constB=0.875}
  n12 = ConstantBiasScale(None=TextureSample(UVs=Add(A=@n11) {constB=0.5}, Tex=NamedRerouteUsage()) {samplerSource=SSM_Clamp_WorldGroupSettings}.RGB) {desc=Normal Map to Linear}
  n13 = Multiply(A=ComponentMask(None=@n12) {g=true}, B=Floor(None=ScalarParameter() {defaultValue=1, parameterName=Show Icon}))
  n14 = NamedRerouteUsage()
  n15 = ScalarParameter() {parameterName=Icon AA Threshold, sliderMax=1}
  n16 = NamedRerouteDeclaration(None=MaterialFunctionCall(
    DistanceField=NamedRerouteDeclaration(None=@n13) {name=Icon DF, nodeColor=<rgba 0 0 0 1>}
    Threshold=LinearInterpolate(A=@n14, Alpha=@n15)
    Smoothing=Constant() {r=0.025}
   ) {desc=Icon, materialFunction=&MF_UI_AntiAliasing}.Outer) {name=Icon Mask, nodeColor=<rgba 0 0 0 1>}
  n17 = StaticSwitchParameter(
   True=LinearInterpolate(A=LinearInterpolate(A=@n2, B=@n9, Alpha=@n10), B=@n9, Alpha=NamedRerouteUsage())
   False=LinearInterpolate(A=@n9, B=LinearInterpolate(A=@n7, B=@n6, Alpha=@n8), Alpha=@n16) {constA=1, constB=0}
  ) {parameterName=Monochromatic}
  n18 = NamedRerouteUsage()
  n19 = VertexColor()
  n20 = Floor(None=ScalarParameter() {parameterName=Use Override, sliderMax=1})
  LinearInterpolate(A=Multiply(A=LinearInterpolate(A=@n1, B=@n17, Alpha=@n18), B=@n19), B=NamedRerouteUsage(), Alpha=@n20) {constAlpha=0}
  n21 = NamedRerouteUsage()
  n22 = Constant() {r=0.3}
  n23 = Constant() {r=1}
  n24 = MaterialFunctionCall("Value (0-1)"=Saturate(None=MaterialFunctionCall(Input=@n21, Input Low=@n22, Input High=@n23) {materialFunction=&RemapValue_0-1}.Result)) {desc=Drop Shadow, materialFunction=&MF_UI_EaseCurves}
  n25 = NamedRerouteDeclaration(None=ComponentMask(None=@n12) {r=true}) {name=Shell DF, nodeColor=<rgba 0 0 0 1>}
  n26 = Constant() {r=0.075}
  n27 = LinearInterpolate(
   A=Multiply(
    A=NamedRerouteDeclaration(None=@n24.EaseInOutCubic) {name=Outline Mask, nodeColor=<rgba 0 0 0 1>}
    B=ScalarParameter() {defaultValue=0.25, desc=Outline Opacity, parameterName=Outline Opacity, sliderMax=1}
   )
   B=ScalarParameter() {defaultValue=1, parameterName=Shell Opacity, sliderMax=1}
   Alpha=NamedRerouteDeclaration(None=MaterialFunctionCall(DistanceField=@n25, Threshold=@n14, Smoothing=@n26) {desc=Shell, materialFunction=&MF_UI_AntiAliasing}.Outer) {name=Shell Mask, nodeColor=<rgba 0 0 0 1>}
  )
  n28 = ScalarParameter() {defaultValue=1, parameterName=Icon Opacity, sliderMax=1}
  LinearInterpolate(A=Multiply(A=@n19.A, B=LinearInterpolate(A=@n27, B=@n28, Alpha=@n16)), B=NamedRerouteUsage(), Alpha=@n20) {constAlpha=0}
  n29 = Constant() {r=0.2}
  MaterialFunctionCall(DistanceField=Subtract(A=NamedRerouteUsage(), B=Multiply(A=@n29) {constB=-1}) {constB=0.1}, Threshold=NamedRerouteUsage()) {desc=Outline Stroke, materialFunction=&MF_UI_AntiAliasing}
  NamedRerouteDeclaration(None=ScalarParameter() {parameterName=Use Dark, sliderMax=1}) {name=Is Dark Mode, nodeColor=<rgba 0 0 0 1>}
  NamedRerouteDeclaration(None=MaterialFunctionCall() {materialFunction=&GetUserInterfaceUV}.Normalized UV) {name=Normalized UV, nodeColor=<rgba 0 0 0 1>}
  NamedRerouteDeclaration(None=TextureObjectParameter() {parameterName=Icon, samplerType=SAMPLERTYPE_Normal, texture=/Engine/EngineMaterials/BaseFlattenNormalMap}) {name=Icon Texture, nodeColor=<rgba 0 0 0 1>}
  n30 = TextureObjectParameter() {parameterName=Icon Override, samplerType=SAMPLERTYPE_LinearColor, texture=/FortUILibrary/Input/Keyboard/Source/Overrides/T_PC_ESC}
  n31 = TextureSample(UVs=NamedRerouteUsage(), Tex=@n30) {samplerType=SAMPLERTYPE_LinearColor, texture=&T_PC_ESC}
  NamedRerouteDeclaration(None=@n31.A) {name=Override Alpha, nodeColor=<rgba 1 0 0.3047 1>}
  NamedRerouteDeclaration(None=@n31.RGB) {name=Override Color, nodeColor=<rgba 0 0.625 1 1>}
  NamedRerouteDeclaration(None=Constant3Vector() {constant=<rgba 1 1 1 0>}) {name=Light Color, nodeColor=<rgba 0 0 0 1>}
  NamedRerouteDeclaration(None=VectorParameter() {parameterName=DarkModeColor, defaultValue=<rgba 0.0048 0.004 0.0116 1>}.RGB) {name=Dark Color, nodeColor=<rgba 0 0 0 1>}
  NamedRerouteDeclaration(None=ScalarParameter() {defaultValue=0.68, desc=Aliasing Threshold, parameterName=AA, sliderMax=1}) {name=Aliasing Threshold, nodeColor=<rgba 0 0 0 1>}
  NamedRerouteDeclaration(None=Constant3Vector() {constant=<rgba 0 0 0 1>}) {name=Outline Color, nodeColor=<rgba 0 0 0 1>}
  NamedRerouteUsage()
 ]
}
